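# Helios_WhiteTemp_001 is a Default Lit material in a test map folder, apparently a temporary placeholder that gives a flat light grey surface.
# Constant ambient, diffuse, specular and shininess values feed the Interchange MF_PhongToMetalRoughness function, which converts Phong inputs to metal-roughness shading.
# The Interchange import data suggests it was generated on asset import.
# It is flagged for use with Nanite and has no textures or parameters.
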Material /BRMapCh6/Maps/Test/Mike/Helios_WhiteTemp_001
graph {
 properties {assetImportData="<self>:InterchangeAssetImportData_0", bUsedWithNanite=true, usedShadingModels=Default Lit}
 nodes [MaterialFunctionCall!(
   AmbientColor=Constant3Vector() {constant=<rgba 0 0 0 1>}
   DiffuseColor=Constant3Vector() {constant=<rgba 0.2188 0.2188 0.2153 1>}
   SpecularColor=Constant3Vector() {constant=<rgba 0.1094 0.1094 0.1076 1>}
   Shininess=Constant() {r=20}
  ) {materialFunction=/InterchangeAssets/Functions/MF_PhongToMetalRoughness}]
}
